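# Masked, unlit, two-sided material for camera-facing firework event sprites on Niagara mesh particles.
# World position offset sums Sprite billboarding, MainSize scaling with optional UV-based stretch, and animated vector noise curl, zeroed on low quality, forward and mobile.
# Emissive multiplies HDR-remapped particle color by sine flicker blended through vertex color channels.
# A circular sphere mask times particle alpha feeds dithered temporal AA opacity.
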Material /Game/Effects/Fort_Effects/Materials/Firework/M_Firework_Event_Sprites_CamFacing_Masked
graph {
 properties {
  bUsedWithNiagaraMeshParticles=true
  blendMode=BLEND_Masked
  pixelDepthOffsetMode=PDOM_Legacy
  shadingModel=MSM_Unlit
  twoSided=true
  usedShadingModels=Unlit
 }
 nodes [
  n1 = ParticlePositionWS()
  n2 = DynamicParameter() {paramNames [X_Size,Y_Size,MainSize,Flicker]}
  n3 = AppendVector(A=@n2.X_Size, B=@n2.X_Size)
  n4 = ScalarParameter() {defaultValue=1, group=Movement, parameterName=Curl_Amount}
  n5 = LinearInterpolate(A=@n4, B=Multiply(A=@n4) {constA=0.65, constB=0.5}, Alpha=PerInstanceRandom())
  n6 = ObjectPositionWS()
  n7 = Subtract(A=WorldPosition().XYZ, B=@n6)
  n8 = ScalarParameter() {defaultValue=50, group=Movement, parameterName=Curl_Speed}
  n9 = Multiply(A=LinearInterpolate(A=@n8, B=Multiply(A=@n8) {constA=0.65, constB=0.65}, Alpha=PerInstanceRandom()), B=Time())
  n10 = ScalarParameter() {defaultValue=2500, group=Movement, parameterName=Curl_Scale}
  n11 = Multiply(A=@n10) {constA=0.65, constB=0.5}
  n12 = PerInstanceRandom()
  n13 = VectorNoise(World Position=Divide(A=Subtract(A=@n7, B=@n9), B=LinearInterpolate(A=@n10, B=@n11, Alpha=@n12))) {noiseFunction=VNF_VectorALU}
  n14 = Constant()
  n15 = FeatureLevelSwitch(
   Default=LinearInterpolate(
    B=Multiply(A=@n5, B=@n13)
    Alpha=DynamicParameter() {parameterIndex=1, defaultValue=<rgba 0 1 1 1>, paramNames [CurlAmount,Stretch,Param3,Param4]}.CurlAmount
   )
   ES2=@n14
   ES3_1=@n14
  )
  n16 = QualitySwitch(Default=@n15, Low=@n14, Medium=@n15)
  n17 = TextureCoordinate()
  n18 = LinearInterpolate(
   A=OneMinus(None=ComponentMask(None=@n17) {g=true})
   B=Constant() {r=1}
   Alpha=DynamicParameter() {parameterIndex=1, defaultValue=<rgba 0 1 1 1>, paramNames [CurlAmount,Stretch,Param3,Param4]}.Stretch
  )
  n19 = Constant() {r=1}
  n20 = Subtract(A=@n1, B=WorldPosition().XYZ)
  Add(
   A=Add(
    A=MaterialFunctionCall!(Center Location=@n1, 0-1 UVs, XY scale=@n3, OverridePivot, "Normals (Optional)") {materialFunction=&Sprite}.World position offset
    B=ShadingPathSwitch(Default=@n16, Forward=@n14, Mobile=@n14)
   )
   B=Multiply(A=StaticSwitchParameter(True=@n18, False=@n19) {defaultValue=true, parameterName=UseDiretionalStretch?}, B=Multiply(A=@n20, B=@n2.MainSize))
  )
  n21 = DynamicParameter() {paramNames [X_Size,Y_Size,MainSize,Flicker]}
  n22 = VertexColor()
  n23 = Clamp(None=LinearInterpolate(A=Sine(None=@n21.Flicker), B=Sine(None=@n21.Flicker) {period=1.5}, Alpha=@n22.B)) {minDefault=0.25}
  n24 = Clamp(None=LinearInterpolate(A=Sine(None=@n21.Flicker) {period=1.1}, B=Sine(None=@n21.Flicker) {period=1.35}, Alpha=@n22.R)) {minDefault=0.25}
  n25 = LinearInterpolate(A=@n22.G, B=@n22.G, Alpha=@n21.Flicker)
  n26 = ParticleColor()
  n27 = Clamp(None=@n26.RGB)
  MaterialFunctionCall(Emissive=Multiply(
    A=LinearInterpolate(A=@n23, B=@n24, Alpha=@n25)
    B=MaterialFunctionCall(LDR Input=@n27, HDR Tint=@n26.RGB) {materialFunction=&DeriveHDRfromLDR}.RemappedRange
   )) {materialFunction=&MF_FX_EmissiveAdjustments}
  n28 = TextureCoordinate()
  n29 = Constant2Vector() {g=0.5, r=0.5}
  n30 = Constant() {r=0.5}
  n31 = Constant() {r=0.5}
  MaterialFunctionCall(Alpha Threshold=Multiply(A=@n26.A, B=SphereMask(A=@n28, B=@n29, Radius=@n30, Hardness=@n31))) {materialFunction=&DitherTemporalAA_Fortnite}
  ParticleColor()
  StaticSwitchParameter(True=ActorPositionWS(), False=@n6) {defaultValue=true, parameterName=ActorPositionWPO}
  WorldPosition()
 ]
}
