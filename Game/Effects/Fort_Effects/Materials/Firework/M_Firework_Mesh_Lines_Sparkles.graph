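# Translucent unlit material for Niagara mesh particles that draws sparkling firework trail lines.
# Opacity multiplies a panned smoke trail by two scrolling noises sharpened via Erode, masked by a V gradient and two dynamic-parameter sphere wipes, with time-of-day scaling and optional depth fade.
# Emissive switches between a curve-atlas ramp and particle color, then applies hue shift.
# World position offset uses SplineThicken for line width, plus optional downward bend and offset.
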Material /Game/Effects/Fort_Effects/Materials/Firework/M_Firework_Mesh_Lines_Sparkles
graph {
 properties {
  bUsedWithNiagaraMeshParticles=true
  blendMode=BLEND_Translucent
  pixelDepthOffsetMode=PDOM_Legacy
  shadingModel=MSM_Unlit
  twoSided=true
  usedShadingModels=Unlit
 }
 nodes [
  ActorPositionWS()
  n1 = ParticleColor()
  n2 = ObjectPositionWS()
  n3 = WorldPosition()
  n4 = ScalarParameter() {defaultValue=100, parameterName=Radius}
  n5 = ScalarParameter() {defaultValue=-10, parameterName=Hardness}
  n6 = LinearInterpolate(
   A=Multiply(B=@n1.RGB) {constA=0.02}
   B=Multiply(B=@n1.RGB) {constA=1}
   Alpha=Clamp(None=SphereMask(A=@n2, B=@n3.XYZ, Radius=@n4, Hardness=@n5))
  )
  Add(A=@n6, B=Multiply(A=@n6, B=DynamicParameter() {defaultValue=<rgba 1 2 0 1>, paramNames [Test,PanSmoke,PanSmokeUp,LineWidth]}.PanSmokeUp))
  ComponentMask(None=TextureCoordinate() {vTiling=0.5}) {g=true}
  Constant2Vector() {g=-0.1}
  Constant2Vector() {r=0.5}
  DynamicParameter() {parameterIndex=3, defaultValue=<rgba 0.1 1 1 1>, paramNames [BaseHueColorShift, XTileSparkle, YTileSparkle, ---]}
  MaterialFunctionCall() {materialFunction=&DebugTimeSine}
  n7 = DynamicParameter() {parameterIndex=3, defaultValue=<rgba 0.1 1 1 1>, paramNames [BaseHueColorShift, XTileSparkle, YTileSparkle, ---]}
  n8 = DynamicParameter() {parameterIndex=1, defaultValue=<rgba 0.1 1 10 1>, paramNames [PannerSpeed, Erode, ColorCurveMult, SparkleMult]}
  n9 = CurveAtlasRowParameter(CurveTime=ComponentMask(None=TextureCoordinate()) {g=true}) {
   atlas=&AthenaCurveAtlas01
   curve=/Game/Athena/LinearColorCurves/Curve_C_01
   defaultValue=201
   parameterName=ColorCurve
  }
  n10 = DynamicParameter() {parameterIndex=1, defaultValue=<rgba 0.1 1 10 1>, paramNames [PannerSpeed, Erode, ColorCurveMult, SparkleMult]}
  n11 = ComponentMask(None=TextureCoordinate()) {g=true}
  n12 = ParticleColor()
  n13 = StaticSwitchParameter(True=Multiply(A=@n9, B=Multiply(A=@n10.ColorCurveMult, B=@n11)), False=Multiply(A=Multiply(A=@n6, B=@n12.RGB), B=@n12.RGB)) {defaultValue=true, parameterName=UseColorCurve?}
  MaterialFunctionCall(Emissive=MaterialFunctionCall(Hue Shift Percentage=@n7.BaseHueColorShift, Texture=Multiply(A=@n8.SparkleMult, B=@n13)) {materialFunction=&HueShift}.Result) {materialFunction=&MF_FX_EmissiveAdjustments}
  n14 = TextureCoordinate() {uTiling=8, vTiling=64}
  n15 = ComponentMask(None=@n14) {r=true}
  n16 = Power(Base=ComponentMask(None=@n14) {g=true}, Exp=ScalarParameter() {defaultValue=0.9, parameterName=Expasdads}) {constExponent=0.8}
  n17 = Constant() {r=4}
  n18 = Multiply(A=DynamicParameter() {defaultValue=<rgba 1 2 0 1>, paramNames [Test,PanSmoke,PanSmokeUp,LineWidth]}.Test) {constB=32}
  n19 = Constant() {r=0.15}
  Multiply(
   A=SphereMask(A=AppendVector(A=@n15, B=@n16), B=AppendVector(A=@n17, B=@n18), Radius=Multiply(A=@n19, B=@n18), Hardness=Constant() {r=-15})
   B=@n6
  )
  n20 = OneMinus(None=Constant() {r=1})
  n21 = ComponentMask(None=VectorParameter() {parameterName=Param_2, defaultValue=<rgba 1 1 0 1>}.RGB) {g=true, r=true}
  n22 = Add(A=AppendVector(A=Constant() {r=0.1}, B=VertexColor().A), B=TextureCoordinate())
  n23 = DynamicParameter() {parameterIndex=1, defaultValue=<rgba 0.1 1 10 1>, paramNames [PannerSpeed, Erode, ColorCurveMult, SparkleMult]}
  n24 = Panner(
   Coordinate=MaterialFunctionCall(UVs=Multiply(A=@n21, B=@n22), "Rotation Angle (0-1)"=ScalarParameter() {defaultValue=0.25}) {materialFunction=&CustomRotator}.Rotated Values
   Time=@n23.PannerSpeed
   Speed=AppendVector(A=@n23.PannerSpeed, B=Constant())
  ) {speedX=-0.025}
  n25 = Constant2Vector() {g=0.1, r=1}
  n26 = Multiply(A=TextureSample(UVs=Multiply(A=@n24, B=@n25)) {samplerType=SAMPLERTYPE_LinearColor, texture=/Game/Effects/Fort_Effects/Textures/Plasma/T_Turbulence_AdditiveWarp_Linear}.R) {constB=0.3}
  n27 = DynamicParameter() {parameterIndex=1, defaultValue=<rgba 0.1 1 10 1>, paramNames [PannerSpeed, Erode, ColorCurveMult, SparkleMult]}
  n28 = OneMinus(None=Subtract(A=Add(A=@n20, B=@n26), B=LinearInterpolate(Alpha=@n27.Erode) {constA=-1}) {constB=-0.7})
  n29 = ScalarParameter() {defaultValue=1, parameterName=XTileSparkle}
  n30 = ScalarParameter() {defaultValue=1, parameterName=YTileSparkle}
  n31 = Add(A=Multiply(A=AppendVector(A=@n29, B=@n30), B=TextureCoordinate() {uTiling=0.25, vTiling=2}), B=VertexColor().A)
  n32 = DynamicParameter() {defaultValue=<rgba 1 2 0 1>, paramNames [Test,PanSmoke,PanSmokeUp,LineWidth]}
  n33 = TextureSample(UVs=Panner(Coordinate=Multiply(A=@n31, B=@n32.Test), Speed=Constant2Vector() {g=0.4, r=-0.003}) {speedX=-0.003, speedY=0.25}) {texture=/Game/Global/Textures/Default/Noise/T_TileNoise_02}
  n34 = TextureCoordinate() {uTiling=0.25, vTiling=2}
  n35 = Multiply(A=Add(A=@n34, B=VertexColor().A), B=DynamicParameter() {defaultValue=<rgba 1 2 0 1>, paramNames [Test,PanSmoke,PanSmokeUp,LineWidth]}.Test)
  n36 = TextureSample(UVs=Panner(Coordinate=Multiply(A=@n32.Test, B=@n35), Speed=Constant2Vector() {g=0.5, r=0.01}) {speedX=0.01, speedY=0.25}) {texture=&T_TileNoise_02}
  n37 = Power(
   Base=Multiply(A=Multiply(A=@n33.G, B=@n36.G)) {constB=100}
   Exp=LinearInterpolate(A=Constant() {r=4}, B=Constant() {r=1.25}, Alpha=@n27.Erode)
  )
  n38 = Multiply(
   A=TextureSample(UVs=@n24) {texture=/Game/Effects/Fort_Effects/Textures/Smoke/T_Puffy_Smoke_Trail_01}.G
   B=Multiply(A=Constant() {r=2}, B=Clamp(None=@n37))
  )
  Multiply(A=LinearInterpolate(B=Constant() {r=0.2}, Alpha=Power(Base=@n28)), B=@n38)
  Multiply(
   A=@n34
   B=AppendVector(A=ScalarParameter() {defaultValue=1, parameterName=XTileSparkle}, B=ScalarParameter() {defaultValue=1, parameterName=YTileSparkle})
  )
  n39 = ComponentMask(None=TextureCoordinate() {vTiling=0.5}) {g=true}
  n40 = ScalarParameter() {defaultValue=0.25, parameterName=Exp}
  n41 = Multiply(
   A=LinearInterpolate(A=Constant() {r=1}, B=Constant(), Alpha=Power(Base=@n39, Exp=@n40) {constExponent=0.8})
   B=VectorParameter() {defaultValue=<rgba 0 0 -2000 1>}.RGB
  )
  n42 = DynamicParameter() {defaultValue=<rgba 1 2 0 1>, paramNames [Test,PanSmoke,PanSmokeUp,LineWidth]}
  n43 = MaterialFunctionCall(
   WidthBase=Multiply(A=ScalarParameter() {defaultValue=250, parameterName=WidthBase}, B=@n42.LineWidth)
   WidthTip=Multiply(A=ScalarParameter() {defaultValue=100, parameterName=WidthTip}, B=@n42.LineWidth)
  ) {materialFunction=&SplineThicken}
  n44 = VectorParameter() {parameterName=OffsetDistance, defaultValue=<rgba -15 -15 -15 0>}
  n45 = VectorParameter() {parameterName=OffsetDistanceMax, defaultValue=<rgba 15 15 15 0>}
  n46 = DynamicParameter() {parameterIndex=2, defaultValue=<rgba 1 2 0 1>, paramNames [Test,PanSmoke,PanSmokeUp,LineWidth]}
  n47 = LinearInterpolate(A=@n44.R, B=@n45.R, Alpha=@n46.Test)
  n48 = LinearInterpolate(A=@n44.G, B=@n45.G, Alpha=@n46.Test)
  n49 = Multiply(
   A=AppendVector(A=AppendVector(A=@n47, B=@n48), B=LinearInterpolate(A=@n44.B, B=@n45.B, Alpha=@n46.Test))
   B=ComponentMask(None=TextureCoordinate()) {g=true}
  )
  n50 = ScalarParameter() {defaultValue=1, parameterName=B}
  StaticSwitchParameter(True=Add(A=Add(A=@n41, B=@n43.WorldPositionOffset), B=Multiply(A=@n49, B=@n50)), False=@n43.WorldPositionOffset) {defaultValue=true, parameterName=UseWPOBend?}
  n51 = ScalarParameter() {defaultValue=1, parameterName="Value_Morning (S)"}
  n52 = ScalarParameter() {defaultValue=2, parameterName="Value_Day (S)"}
  n53 = ScalarParameter() {defaultValue=1, parameterName="Value_Evening (S)"}
  n54 = ScalarParameter() {defaultValue=1, parameterName="Value_Night (S)"}
  n55 = MaterialFunctionCall(UV Channel=TextureCoordinate()) {materialFunction=&LinearGradient}
  n56 = Multiply(A=@n38, B=Multiply(A=OneMinus(None=@n55.VGradient), B=@n55.VGradient))
  n57 = ParticleColor()
  n58 = TextureCoordinate() {uTiling=0}
  n59 = AppendVector(A=Constant(), B=DynamicParameter() {defaultValue=<rgba 1 2 0 1>, paramNames [Test,PanSmoke,PanSmokeUp,LineWidth]}.PanSmoke)
  n60 = Constant() {r=1}
  n61 = Constant() {r=0.8}
  n62 = TextureCoordinate() {uTiling=0, vTiling=-1}
  n63 = AppendVector(A=Constant(), B=DynamicParameter() {defaultValue=<rgba 1 2 0 1>, paramNames [Test,PanSmoke,PanSmokeUp,LineWidth]}.PanSmokeUp)
  n64 = Multiply(
   A=Multiply(
    A=MaterialFunctionCall(Value_Morning=@n51, Value_Day=@n52, Value_Evening=@n53, Value_Night=@n54) {materialFunction=&Scalar_Time_of_Day}.Result
    B=Multiply(A=@n56, B=@n57.A)
   )
   B=Multiply(A=SphereMask(A=@n58, B=@n59, Radius=@n60, Hardness=@n61), B=SphereMask(A=@n62, B=@n63, Radius=@n60, Hardness=@n61))
  )
  n65 = DepthFade(Opacity=@n64, FadeDistance=ScalarParameter() {parameterName=DepthFadeDistance})
  StaticSwitchParameter(True=ShadingPathSwitch(Default=QualitySwitch(Default=@n65, Low=@n64, Medium=@n64), Forward=@n64, Mobile=@n64), False=@n64) {parameterName=UseDepthFade?}
  TextureSample(UVs=@n24) {texture=&T_Puffy_Smoke_Trail_01}
 ]
}
