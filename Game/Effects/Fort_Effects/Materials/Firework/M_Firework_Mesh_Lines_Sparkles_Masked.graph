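# Masked, unlit, two-sided material for Niagara mesh particles that draws sparkling firework streaks, driven by dynamic parameters.
# Color comes from particle color and an optional curve-atlas gradient, then hue shift and emissive adjustments.
# A panned smoke texture, two multiplied scatter textures with erosion, gradient and sphere masks, and a time-of-day scalar feed a temporal-dithered mask.
# SplineThicken sets line width through world position offset, with an optional bend and offset along V.
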Material /Game/Effects/Fort_Effects/Materials/Firework/M_Firework_Mesh_Lines_Sparkles_Masked
graph {
 properties {
  bUsedWithNiagaraMeshParticles=true
  blendMode=BLEND_Masked
  pixelDepthOffsetMode=PDOM_Legacy
  shadingModel=MSM_Unlit
  twoSided=true
  usedShadingModels=Unlit
 }
 nodes [
  Constant2Vector() {g=-0.1}
  n1 = DynamicParameter() {parameterIndex=3, defaultValue=<rgba 0.1 1 1 1>, paramNames [BaseHueColorShift, XTileSparkle, YTileSparkle, BentUporDown]}
  n2 = DynamicParameter() {parameterIndex=1, defaultValue=<rgba 0.1 1 10 1>, paramNames [PannerSpeed, Erode, ColorCurveMult, SparkleMult]}
  n3 = CurveAtlasRowParameter(CurveTime=ComponentMask(None=TextureCoordinate()) {g=true}) {
   atlas=&AthenaCurveAtlas01
   curve=/Game/Athena/LinearColorCurves/Curve_C_01
   defaultValue=201
   parameterName=ColorCurve
  }
  n4 = DynamicParameter() {parameterIndex=1, defaultValue=<rgba 0.1 1 10 1>, paramNames [PannerSpeed, Erode, ColorCurveMult, SparkleMult]}
  n5 = Multiply(A=@n4.ColorCurveMult, B=ComponentMask(None=TextureCoordinate()) {g=true})
  n6 = StaticSwitchParameter(True=Multiply(A=Multiply(A=@n3, B=@n5), B=ParticleColor().RGB), False=ParticleColor().RGB) {defaultValue=true, parameterName=UseColorCurve?}
  MaterialFunctionCall(Emissive=MaterialFunctionCall(Hue Shift Percentage=@n1.BaseHueColorShift, Texture=Multiply(A=@n2.SparkleMult, B=@n6)) {materialFunction=&HueShift}.Result) {materialFunction=&MF_FX_EmissiveAdjustments}
  n7 = MaterialFunctionCall(
   Value_Morning=ScalarParameter() {defaultValue=1, parameterName="Value_Morning (S)"}
   Value_Day=ScalarParameter() {defaultValue=2, parameterName="Value_Day (S)"}
   Value_Evening=ScalarParameter() {defaultValue=1, parameterName="Value_Evening (S)"}
   Value_Night=ScalarParameter() {defaultValue=1, parameterName="Value_Night (S)"}
  ) {materialFunction=&Scalar_Time_of_Day}
  n8 = Constant() {r=0.1}
  n9 = VertexColor()
  n10 = Multiply(
   A=ComponentMask(None=VectorParameter() {parameterName=Param_2, defaultValue=<rgba 1 1 0 1>}.RGB) {g=true, r=true}
   B=Add(A=AppendVector(A=@n8, B=@n9.A), B=TextureCoordinate())
  )
  n11 = ScalarParameter() {defaultValue=0.25}
  n12 = DynamicParameter() {parameterIndex=1, defaultValue=<rgba 0.1 1 10 1>, paramNames [PannerSpeed, Erode, ColorCurveMult, SparkleMult]}
  n13 = Constant()
  n14 = TextureSample(UVs=Panner(
    Coordinate=MaterialFunctionCall(UVs=@n10, "Rotation Angle (0-1)"=@n11) {materialFunction=&CustomRotator}.Rotated Values
    Time=@n12.PannerSpeed
    Speed=AppendVector(A=@n12.PannerSpeed, B=@n13)
   ) {speedX=-0.025}) {texture=/Game/Effects/Fort_Effects/Textures/Smoke/T_Puffy_Smoke_Trail_01}
  n15 = DynamicParameter() {parameterIndex=3, defaultValue=<rgba 0.1 1 1 1>, paramNames [BaseHueColorShift, XTileSparkle, YTileSparkle, BentUporDown]}
  n16 = Multiply(A=AppendVector(A=@n15.XTileSparkle, B=@n15.XTileSparkle), B=TextureCoordinate() {uTiling=0.25, vTiling=2})
  n17 = VertexColor()
  n18 = DynamicParameter() {defaultValue=<rgba 1 2 0 1>, paramNames [Test,PanSmoke,PanSmokeUp,LineWidth]}
  n19 = Panner(Coordinate=Multiply(A=Add(A=@n16, B=@n17.A), B=@n18.Test), Speed=Constant2Vector() {g=0.4, r=-0.003}) {speedX=-0.003, speedY=0.25}
  n20 = TextureCoordinate() {uTiling=0.25, vTiling=2}
  n21 = Add(A=@n20, B=VertexColor().A)
  n22 = DynamicParameter() {defaultValue=<rgba 1 2 0 1>, paramNames [Test,PanSmoke,PanSmokeUp,LineWidth]}
  n23 = Panner(Coordinate=Multiply(A=@n18.Test, B=Multiply(A=@n21, B=@n22.Test)), Speed=Constant2Vector() {g=0.5, r=0.01}) {speedX=0.01, speedY=0.25}
  n24 = Multiply(A=Multiply(
    A=TextureSampleParameter2D(UVs=@n19) {parameterName=Tex1, texture=/Game/Environments/Helios/FX/OpenDesert/Textures/T_Spheres_Scatter_01}.G
    B=TextureSampleParameter2D(UVs=@n23) {parameterName=Tex2, texture=&T_Spheres_Scatter_01}.G
   )) {constB=5}
  n25 = DynamicParameter() {parameterIndex=1, defaultValue=<rgba 0.1 1 10 1>, paramNames [PannerSpeed, Erode, ColorCurveMult, SparkleMult]}
  n26 = LinearInterpolate(A=Constant() {r=4}, B=Constant() {r=1.25}, Alpha=@n25.Erode)
  n27 = Multiply(A=Constant() {r=2}, B=Clamp(None=Power(Base=@n24, Exp=@n26)))
  n28 = MaterialFunctionCall(UV Channel=TextureCoordinate()) {materialFunction=&LinearGradient}
  n29 = OneMinus(None=@n28.VGradient)
  n30 = Multiply(A=Multiply(A=Multiply(A=@n14.G, B=@n27), B=Multiply(A=@n29, B=@n28.VGradient)), B=ParticleColor().A)
  n31 = Constant() {r=1}
  n32 = Constant() {r=0.8}
  n33 = SphereMask(
   A=TextureCoordinate() {uTiling=0}
   B=AppendVector(A=Constant(), B=DynamicParameter() {defaultValue=<rgba 1 2 0 1>, paramNames [Test,PanSmoke,PanSmokeUp,LineWidth]}.PanSmoke)
   Radius=@n31
   Hardness=@n32
  )
  n34 = SphereMask(
   A=TextureCoordinate() {uTiling=0, vTiling=-1}
   B=AppendVector(A=Constant(), B=DynamicParameter() {defaultValue=<rgba 1 2 0 1>, paramNames [Test,PanSmoke,PanSmokeUp,LineWidth]}.PanSmokeUp)
   Radius=@n31
   Hardness=@n32
  )
  MaterialFunctionCall(Alpha Threshold=Multiply(A=Multiply(A=@n7.Result, B=@n30), B=Multiply(A=@n33, B=@n34))) {materialFunction=&DitherTemporalAA_Fortnite}
  Multiply(A=@n20, B=AppendVector(A=@n15.XTileSparkle, B=@n15.XTileSparkle))
  ScalarParameter() {defaultValue=0.25, parameterName=Exp}
  n35 = ComponentMask(None=TextureCoordinate() {vTiling=0.5}) {g=true}
  n36 = DynamicParameter() {parameterIndex=3, defaultValue=<rgba 0.1 1 1 1>, paramNames [BaseHueColorShift, XTileSparkle, YTileSparkle, BentUporDown]}
  n37 = Multiply(
   A=LinearInterpolate(A=Constant() {r=1}, B=Constant(), Alpha=Power(Base=@n35, Exp=@n36.BentUporDown) {constExponent=0.8})
   B=VectorParameter() {defaultValue=<rgba 0 0 500 1>}.RGB
  )
  n38 = DynamicParameter() {defaultValue=<rgba 1 2 0 1>, paramNames [Test,PanSmoke,PanSmokeUp,LineWidth]}
  n39 = MaterialFunctionCall(
   WidthBase=Multiply(A=ScalarParameter() {defaultValue=250, parameterName=WidthBase}, B=@n38.LineWidth)
   WidthTip=Multiply(A=ScalarParameter() {defaultValue=100, parameterName=WidthTip}, B=@n38.LineWidth)
  ) {materialFunction=&SplineThicken}
  n40 = VectorParameter() {parameterName=OffsetDistance, defaultValue=<rgba -15 -15 -15 0>}
  n41 = VectorParameter() {parameterName=OffsetDistanceMax, defaultValue=<rgba 15 15 15 0>}
  n42 = DynamicParameter() {parameterIndex=2, defaultValue=<rgba 1 2 0 1>, paramNames [Test,PanSmoke,PanSmokeUp,LineWidth]}
  n43 = LinearInterpolate(A=@n40.R, B=@n41.R, Alpha=@n42.Test)
  n44 = LinearInterpolate(A=@n40.G, B=@n41.G, Alpha=@n42.Test)
  n45 = Multiply(
   A=AppendVector(A=AppendVector(A=@n43, B=@n44), B=LinearInterpolate(A=@n40.B, B=@n41.B, Alpha=@n42.Test))
   B=ComponentMask(None=TextureCoordinate()) {g=true}
  )
  n46 = ScalarParameter() {defaultValue=1, parameterName=B}
  StaticSwitchParameter(True=Add(A=Add(A=@n37, B=@n39.WorldPositionOffset), B=Multiply(A=@n45, B=@n46)), False=@n39.WorldPositionOffset) {defaultValue=true, parameterName=UseWPOBend?}
 ]
}
