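# Translucent, unlit, two-sided firework material that draws glowing flare sprites on meshes, including Niagara mesh particles.
# The Sprite function builds position offsets sized by vertex alpha and dynamic parameters, alongside a transformed offset scaled by vertex red.
# Particle color is boosted through emissive adjustments, while a flare texture, sphere mask and particle and vertex alpha shape opacity.
# Depth fade is skipped on forward, mobile and low or medium quality paths.
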Material /Game/Effects/Fort_Effects/Materials/Firework/M_Firework_Mesh_Sprites
graph {
 properties {
  bCastRayTracedShadows=false
  bUsedWithInstancedStaticMeshes=true
  bUsedWithNiagaraMeshParticles=true
  bUsedWithSplineMeshes=true
  blendMode=BLEND_Translucent
  pixelDepthOffsetMode=PDOM_Legacy
  shadingModel=MSM_Unlit
  twoSided=true
  usedShadingModels=Unlit
 }
 nodes [
  n1 = VertexColor()
  n2 = DynamicParameter() {parameterIndex=3, defaultValue=<rgba 1 1 -1 1>, paramNames [BaseColorHueShift, SpriteSize, SpriteSizeSmall, --------]}
  n3 = ConstantBiasScale(None=@n1.A) {bias=-0.2, scale=1.2}
  n4 = MaterialFunctionCall!(
   Center Location=WorldPosition().XYZ
   0-1 UVs
   XY scale=Multiply(A=@n1.A, B=LinearInterpolate(A=@n2.SpriteSizeSmall, B=@n2.SpriteSizeSmall, Alpha=@n3))
   OverridePivot
   "Normals (Optional)"
  ) {materialFunction=&Sprite}
  n5 = VectorParameter() {parameterName=OffsetDistance, defaultValue=<rgba 0 2 0 0>}
  n6 = VertexColor()
  n7 = Multiply(
   A=Transform(None=Multiply(A=@n5.RGB, B=@n6.R))
   B=LinearInterpolate(
    B=ScalarParameter() {defaultValue=2, parameterName=OffsetDistance}
    Alpha=DynamicParameter() {defaultValue=<rgba 1 1 -1 1>, paramNames [Test,PanSmoke,OffsetMesh,LineWidth]}.OffsetMesh
   ) {constB=150}
  )
  Add(A=@n4.World position offset, B=@n7)
  Add(A=@n7)
  n8 = PixelDepth()
  n9 = ScalarParameter() {defaultValue=0.1, desc=Multiplier for adding spline width based on size on screen, group=SplineWidth, parameterName=ScreenWidthMult}
  n10 = Multiply(A=ComponentMask(None=ViewSize()) {g=true}, B=@n9)
  n11 = ScalarParameter() {defaultValue=5, desc=Max clamp on how much screen size can add to spline width, group=SplineWidth, parameterName=ScreenWidthMaxAdd}
  Add(A=Clamp(None=Divide(A=@n8.R, B=@n10), Max=@n11))
  n12 = ParticleColor()
  MaterialFunctionCall(Emissive=Multiply(A=Constant() {r=100}, B=@n12.RGB)) {materialFunction=&MF_FX_EmissiveAdjustments}
  Multiply(A=@n4.World position offset, B=@n7)
  Power(Base=ComponentMask(None=TextureCoordinate() {vTiling=4}) {g=true}) {constExponent=1}
  n13 = SphereMask(A=TextureCoordinate(), B=Constant2Vector() {g=0.5, r=0.5}, Radius=Constant() {r=0.5}, Hardness=Constant())
  n14 = Multiply(
   A=VertexColor().A
   B=Multiply(
    A=Multiply(A=@n12.A, B=@n13)
    B=TextureSampleParameter2D() {parameterName=FlareMask, texture=/Game/Effects/Fort_Effects/Textures/Flares/T_Flare_Spikeball}.G
   )
  )
  n15 = Multiply(A=@n14, B=DepthFade(FadeDistance=ScalarParameter() {defaultValue=100, parameterName=FadeDistance}))
  ShadingPathSwitch(Default=QualitySwitch(Default=@n15, Low=@n14, High=@n15, Medium=@n14, Epic=@n15), Forward=@n14, Mobile=@n14)
  Subtract(A=@n1.A) {constB=0.2}
 ]
}
