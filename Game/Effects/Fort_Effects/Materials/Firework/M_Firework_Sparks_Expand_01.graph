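# Unlit translucent, two-sided material for Niagara mesh particles that draws expanding firework spark streaks with HDR emissive from particle color.
# Opacity combines a streak texture, vertex and particle color, and a panning-noise, sphere-masked glow tip with dynamic-parameter erosion.
# World position offset slides vertices along a transformed offset and uses SplineThicken or Sprite functions, with a SwitchToSprite blend.
# Spline width grows with pixel depth relative to viewport height, up to a clamp.
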Material /Game/Effects/Fort_Effects/Materials/Firework/M_Firework_Sparks_Expand_01
graph {
 properties {
  bEnableResponsiveAA=true
  bUseTranslucencyVertexFog=false
  bUsedWithNiagaraMeshParticles=true
  blendMode=BLEND_Translucent
  pixelDepthOffsetMode=PDOM_Legacy
  shadingModel=MSM_Unlit
  twoSided=true
  usedShadingModels=Unlit
 }
 nodes [
  n1 = VertexColor()
  n2 = NamedRerouteUsage()
  n3 = Constant()
  n4 = DynamicParameter() {defaultValue=<rgba 1 1 -1 1>, paramNames [Test,PanSmoke,OffsetMesh,LineWidth]}
  n5 = SphereMask(
   A=TextureCoordinate() {uTiling=0}
   B=ComponentMask(None=AppendVector(A=@n3, B=@n4.PanSmoke)) {g=true, r=true}
   Radius=ScalarParameter() {defaultValue=0.8, parameterName=GlowTip_Radius_1}
   Hardness=ScalarParameter() {parameterName=GlowTip_Hardness_1}
  )
  n6 = Panner(Coordinate=Add(A=TextureCoordinate() {uTiling=0.08}, B=VertexColor().A)) {speedY=-0.04}
  n7 = Panner(Coordinate=Add(A=TextureCoordinate() {uTiling=0.1}, B=VertexColor().A)) {speedY=0.08}
  n8 = Multiply(A=Multiply(
    A=TextureSample(UVs=@n6) {texture=/Game/Global/Textures/Default/Noise/T_TileNoise_02}.G
    B=TextureSample(UVs=@n7) {texture=&T_TileNoise_02}.G
   )) {constB=50}
  n9 = TextureCoordinate()
  n10 = DynamicParameter() {parameterIndex=1, defaultValue=<rgba 0.07 1 1 1>, paramNames [Erosion, Erosion_Softness, AnimationSpeed, SwitchToSprite]}
  n11 = Subtract(A=Multiply(A=Clamp(None=@n5), B=Clamp(None=@n8)), B=Multiply(A=ComponentMask(None=@n9) {g=true}, B=@n10.Erosion))
  n12 = DynamicParameter() {parameterIndex=1, defaultValue=<rgba 0.07 1 1 1>, paramNames [Erosion, Erosion_Softness, AnimationSpeed, SwitchToSprite]}
  n13 = Multiply(A=Saturate(None=Divide(A=@n11, B=@n12.Erosion_Softness)), B=Multiply(A=Constant() {r=250}, B=NamedRerouteUsage()))
  n14 = DynamicParameter() {parameterIndex=1, defaultValue=<rgba 0.07 1 1 1>, paramNames [Erosion, Erosion_Softness, AnimationSpeed, SwitchToSprite]}
  Multiply(
   A=Multiply(A=Clamp(None=@n1.R), B=ParticleColor().A)
   B=LinearInterpolate(A=Multiply(A=@n2), B=Multiply(A=@n2, B=@n13), Alpha=@n14.SwitchToSprite)
  )
  n15 = Multiply(A=ComponentMask(None=TextureCoordinate()) {g=true}, B=ParticleColor().RGB)
  n16 = Clamp(None=@n15)
  n17 = Multiply(A=Constant() {r=5}, B=@n15)
  NamedRerouteDeclaration(None=MaterialFunctionCall(Emissive=MaterialFunctionCall(LDR Input=@n16, HDR Tint=@n17) {materialFunction=&DeriveHDRfromLDR}.RemappedRange) {materialFunction=&MF_FX_EmissiveAdjustments}.Result) {name=Emissive, nodeColor=<rgba 1 0 0.9375 1>}
  n18 = NamedRerouteUsage()
  n19 = NamedRerouteUsage()
  n20 = DynamicParameter() {parameterIndex=1, defaultValue=<rgba 0.07 1 1 1>, paramNames [Erosion, Erosion_Softness, AnimationSpeed, SwitchToSprite]}
  NamedRerouteDeclaration(None=Add(A=NamedRerouteUsage(), B=LinearInterpolate(A=@n18, B=@n19, Alpha=@n20.SwitchToSprite))) {name=WPO, nodeColor=<rgba 1 0.4219 0 1>}
  n21 = ViewSize()
  n22 = ScalarParameter() {defaultValue=0.1, desc=Multiplier for adding spline width based on size on screen, group=SplineWidth, parameterName=ScreenWidthMult}
  n23 = Divide(A=PixelDepth().R, B=Multiply(A=ComponentMask(None=@n21) {g=true}, B=@n22))
  n24 = ScalarParameter() {defaultValue=5, desc=Max clamp on how much screen size can add to spline width, group=SplineWidth, parameterName=ScreenWidthMaxAdd}
  NamedRerouteDeclaration(None=Add(A=Clamp(None=@n23, Max=@n24))) {name=MinSizeOnscreen, nodeColor=<rgba 1 0 0.5391 1>}
  n25 = Multiply(A=ScalarParameter() {defaultValue=8, group=SplineWidth, parameterName=WidthBase}, B=NamedRerouteUsage())
  n26 = DynamicParameter() {defaultValue=<rgba 1 1 -1 1>, paramNames [Test,PanSmoke,OffsetMesh,LineWidth]}
  n27 = Multiply(A=ScalarParameter() {defaultValue=8, group=SplineWidth, parameterName=WidthTip}, B=NamedRerouteUsage())
  NamedRerouteDeclaration(None=MaterialFunctionCall(WidthBase=Multiply(A=@n25, B=@n26.LineWidth), WidthTip=Multiply(A=@n26.LineWidth, B=@n27)) {materialFunction=&SplineThicken}.WorldPositionOffset) {name=SplineThicken, nodeColor=<rgba 0 0.8594 1 1>}
  NamedRerouteDeclaration(None=MaterialFunctionCall!(Center Location=WorldPosition().XYZ, 0-1 UVs, XY scale=Constant2Vector() {g=100, r=100}, OverridePivot, "Normals (Optional)") {materialFunction=&Sprite}.World position offset) {name=Sprite, nodeColor=<rgba 1 0.8203 0 1>}
  n28 = TextureCoordinate() {vTiling=4}
  n29 = Multiply(
   A=Power(Base=ComponentMask(None=@n28) {g=true}) {constExponent=1}
   B=Multiply(A=VectorParameter() {parameterName=OffsetDistance, defaultValue=<rgba 0 3 0 0>}.RGB, B=VertexColor().R)
  )
  n30 = ScalarParameter() {defaultValue=2, parameterName=OffsetDistance}
  n31 = DynamicParameter() {defaultValue=<rgba 1 1 -1 1>, paramNames [Test,PanSmoke,OffsetMesh,LineWidth]}
  NamedRerouteDeclaration(None=Multiply(A=Transform(None=@n29), B=LinearInterpolate(B=@n30, Alpha=@n31.OffsetMesh) {constB=150})) {name=SlideAlongtangent, nodeColor=<rgba 1 0 0.7031 1>}
  n32 = Constant() {r=1}
  n33 = Constant() {r=1}
  n34 = TextureSample(UVs=Multiply(A=TextureCoordinate(), B=AppendVector(A=@n32, B=@n33))) {samplerSource=SSM_Wrap_WorldGroupSettings, texture=/Game/Missions/Secondary/Events/HonkingFirework/FX/Textures/T_FireWorks_Capsule_Sparks_01}
  NamedRerouteDeclaration(None=@n34.G) {name=Streak, nodeColor=<rgba 0 1 0.9766 1>}
  NamedRerouteUsage()
  NamedRerouteUsage()
  Power(Base=@n8) {constExponent=1.25}
 ]
}
